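# Unlit additive material for the Soldier War Cry shockwave, used on mesh particles.
# One texture supplies an edge line with adjustable thinness (R) plus a mask (G) multiplied by panning detail (B).
# Dynamic parameters drive vertical pan speed and black-clip erosion; particle alpha scales the result, with optional depth fade bypassed on mobile.
# Emissive adds a warm depth-intersection glow to particle color, then runs through MF_FX_EmissiveAdjustments.
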Material /Game/CharClasses/Soldier/Traits/AdvancedTactics/Abilities/WarCry/M_WarCryShockwave
graph {
 properties {
  bUsedWithMeshParticles=true
  bUsedWithNiagaraMeshParticles=true
  blendMode=BLEND_Additive
  pixelDepthOffsetMode=PDOM_Legacy
  shadingModel=MSM_Unlit
  usedShadingModels=Unlit
 }
 nodes [
  n1 = ParticleColor()
  n2 = TextureSample() {texture=&/T_Commando_WarCry_Shockwave}
  n3 = Clamp(None=Power(Base=@n2.R, Exp=ScalarParameter() {defaultValue=7, parameterName=LineThinness}) {constExponent=7})
  n4 = ScalarParameter() {parameterName=ChangeEdgeThickness?, sliderMax=1}
  n5 = TextureCoordinate()
  n6 = AppendVector(
   A=ScalarParameter() {parameterName=SpeedX}
   B=DynamicParameter() {defaultValue=<rgba 0 0.8 1 1>, paramNames [BlackClip,SpeedY,Param3,Param4]}.SpeedY
  )
  n7 = Multiply(A=TextureSample(UVs=Panner(Coordinate=@n5, Speed=@n6) {speedY=0.8}) {texture=&T_Commando_WarCry_Shockwave}.B) {constB=0.6}
  n8 = Clamp(None=Add(A=LinearInterpolate(A=@n2.R, B=@n3, Alpha=@n4), B=Multiply(A=@n2.G, B=@n7)))
  n9 = DynamicParameter() {defaultValue=<rgba 0 0.8 1 1>, paramNames [BlackClip,SpeedY,Param3,Param4]}
  n10 = Multiply(A=@n1.A, B=Divide(A=Subtract(A=@n8, B=@n9.BlackClip), B=OneMinus(None=@n9.BlackClip)))
  n11 = DepthFade(FadeDistance=ScalarParameter() {parameterName=FadeDistance})
  n12 = Constant() {r=1}
  LinearInterpolate(
   A=@n10
   B=Multiply(A=@n10, B=ShadingPathSwitch(Default=@n11, Deferred=@n11, Mobile=@n12) {desc=Migrated from FeatureLevelSwitch})
   Alpha=ScalarParameter() {parameterName=UseDepthFade?, sliderMax=1}
  )
  n13 = VectorParameter() {defaultValue=<rgba 1 0.852 0.3517 0>}
  n14 = OneMinus(None=DepthFade(Opacity=Constant() {r=1}, FadeDistance=Constant() {r=15}))
  MaterialFunctionCall(Emissive=Add(A=Multiply(A=@n13.RGB, B=@n14), B=@n1.RGB)) {materialFunction=&MF_FX_EmissiveAdjustments}
 ]
}
