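# Unlit additive particle material for the Soldier War Cry buff effect, usable on sprite and Niagara particles.
# A War Cry streak texture, with UVs offset by panning noise scaled over particle lifetime, is multiplied by particle alpha, while particle color feeds an emissive-adjustment function.
# Separate branches compute depth fade, screen-space masking with camera-depth fade, and switchable near-camera fade.
# Quality and feature-level switches bypass depth and near-camera fades on low-end targets.
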Material /Game/CharClasses/Soldier/Traits/AdvancedTactics/Abilities/WarCry/M_Warcry_BuffAdd
graph {
 properties {
  bUsedWithNiagaraSprites=true
  bUsedWithParticleSprites=true
  blendMode=BLEND_Additive
  pixelDepthOffsetMode=PDOM_Legacy
  shadingModel=MSM_Unlit
  usedShadingModels=Unlit
 }
 nodes [
  n1 = ParticleColor()
  MaterialFunctionCall(Emissive=@n1.RGB) {materialFunction=&MF_FX_EmissiveAdjustments}
  n2 = TextureCoordinate()
  n3 = TextureSample(UVs=Panner(Coordinate=TextureCoordinate() {uTiling=0.3}) {speedY=-3}) {texture=&T_NoisePack_01}
  n4 = Constant()
  n5 = ParticleRelativeTime()
  n6 = Multiply(
   A=Multiply(A=MaterialFunctionCall(X=@n3.R, Y=@n4) {materialFunction=&MakeFloat2}.Result) {constB=0.3}
   B=@n5
  )
  n7 = Multiply(
   A=@n1.A
   B=TextureSample(UVs=Add(A=@n2, B=@n6)) {texture=&/T_Commando_WarCry_Streak}.RGB
  )
  n8 = DepthFade(FadeDistance=Constant() {r=12})
  n9 = Constant() {r=1}
  n10 = FeatureLevelSwitch(Default=@n8, ES2=@n9, ES3_1=@n9, SM4_REMOVED=@n8)
  Multiply(A=@n7, B=QualitySwitch(Default=@n10, Low=@n9, High=@n10, Medium=@n10))
  n11 = If(A=ComponentMask(None=ScreenPosition().ViewportUV) {r=true}, B=Constant() {r=0.2}, "A > B"=Constant(), "A < B"=Constant() {r=1})
  n12 = SphereMask(A=Constant(), B=Constant() {r=1}, Radius=Constant() {r=256}, Hardness=Constant() {r=10})
  n13 = MaterialFunctionCall(
   Fade Length=ScalarParameter() {defaultValue=500, parameterName=DepthFade}
   Fade Offset=ScalarParameter() {defaultValue=64, parameterName=FadeOffset}
  ) {materialFunction=&CameraDepthFade}
  Multiply(A=Multiply(A=@n7, B=LinearInterpolate(A=@n11, B=@n12, Alpha=@n13.Result)))
  n14 = TextureSampleParameter2D(UVs=ScreenPosition().ViewportUV) {parameterName=Param, texture=/Game/UI/Frontend/Store/Textures/T_RadGradient}
  Multiply(A=@n14.RGB) {constB=-1}
  OneMinus(None=@n5)
  OneMinus(None=@n14.RGB)
  ScalarParameter() {defaultValue=1}
  n15 = MaterialFunctionCall(Distance=DynamicParameter() {paramNames [Distance,Param2,Param3,Param4]}.Distance) {materialFunction=&Near_Camera_Fade}
  n16 = Constant() {r=1}
  n17 = FeatureLevelSwitch(Default=@n15.Result, ES2=@n16, ES3_1=@n16, SM4_REMOVED=@n15.Result)
  n18 = ScalarParameter() {defaultValue=512, parameterName=FadeLength}
  n19 = ScalarParameter() {defaultValue=64, parameterName=FadeOffset}
  StaticSwitchParameter(
   True=QualitySwitch(Default=@n17, Low=@n16, High=@n17, Medium=@n17)
   False=LinearInterpolate(
    A=ScalarParameter() {defaultValue=0.38, parameterName=NearOpacity}
    Alpha=MaterialFunctionCall(Fade Length=@n18, Fade Offset=@n19) {materialFunction=&CameraDepthFade}.Result
   ) {constA=0.3}
  ) {parameterName=UseNearCameraFade?}
 ]
}
